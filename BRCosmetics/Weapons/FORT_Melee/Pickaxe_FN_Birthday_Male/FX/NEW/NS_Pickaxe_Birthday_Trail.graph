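# Niagara trail effect for the male Birthday pickaxe, combining swing ribbons with flying cake and sprinkle meshes.
# AnimTrail and FlameTrail build socket- or bone-driven ribbons; FlameTrail blends three user colors, tinted by a linked time-of-day value.
# Cake and Sprinkles spawn per unit moved in a box around a socket, inherit velocity, then fall, tumble, shrink and stretch with speed.
# Sprinkles randomly take one of those three user colors.
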NiagaraSystem /BRCosmetics/Weapons/FORT_Melee/Pickaxe_FN_Birthday_Male/FX/NEW/NS_Pickaxe_Birthday_Trail
dependencies {
 usedDataInterfaces [
  &NiagaraDataInterfaceCurve
  &NiagaraDataInterfaceColorCurve
  &NiagaraDataInterfaceSkeletalMesh
  &NiagaraDataInterfaceSocketReader
 ]
 usedDynamicInputs []
 usedModules [
  &self:SystemSpawnScript.SystemScriptSource.SystemScriptGraph.NiagaraNodeAssignment_0.SetVariables_#1
  &SystemState
 ]
}
emitters {
 AnimTrail {
  emitterSpawnScript [&FN_RibbonIDOffsetUtility()]
  emitterUpdateScript [
   &EmitterState(Life Cycle Mode=::NewEnumerator1, Loop Behavior=::NewEnumerator1, Loop Duration=1, UseLoopDelay=false)
   &FN_SpawnAnimTrail(Number of Anim Trails=2, Spawn Interval=0)
   /SetVariables(Emitter.Exp=0.7) {isSetParameters=true}
  ]
  particleSpawnScript [
   &FN_InitializeAnimTrailBoneIndecies(First Bone or Socket Index=2, Second Bone or Socket Index=2)
   &FN_InitializeAnimTrailBoneIndecies(AnimTrail Selection=1, First Bone or Socket Index=3, Second Bone or Socket Index=3) {module=FN_InitializeAnimTrailBoneIndecies001}
   &FN_InitializeAnimTrail(
    AnimTrailWidth=30
    Color=&MakeLinearColorFromVectorAndFloat(
     "Vector (RGB)"=<1 1 1>
     "Float (Alpha)"=&FloatFromCurve(FloatCurve=DataInterface() {propertyValues {Curve {keys [<at 0.17 1>,<at 0.2 0>]}}}, CurveIndex=$Engine.System.Age)
    )
    Lifetime=0.25
   )
   &DynamicMaterialParameters(Module.Index0Param0Write=true, Module.Index0Param1Write=true, Module.Index0Param2Write=true, Module.Index0Param3Write=true)
  ]
  particleUpdateScript [
   &FN_UpdateAnimTrail(Locked Particle RibbonWidth=1) {module=FN_UpdateAnimTrail001}
   &ScaleRibbonWidth(Ribbon Width Scale=&FloatFromCurve(FloatCurve=DataInterface() {propertyValues {Curve {keys [{interpMode=RCIM_Cubic, tangentMode=RCTM_User, time=0.5, value=1}, {arriveTangent=-3.7754, tangentMode=RCTM_User, time=0.9, value=0}]}}}))
   &ScaleColor(
    Scale Alpha=&FloatFromCurve(
     FloatCurve=DataInterface() {propertyValues {Curve {keys [<at 0.09 1 RCIM_Cubic>,<at 0.125 0 RCIM_Cubic>]}}}
     CurveIndex=$Particles.Age
     Scale Curve=&FloatFromCurve(
      FloatCurve=DataInterface() {propertyValues {Curve {keys [<at 0 0>, {tangentMode=RCTM_User, time=0.1, value=1}]}}}
      CurveIndex=$Emitter.NormalizedLoopAge
     )
    )
    Scale RGB=<5 5 5>
    ScaleA=true
    ScaleRGB=false
   )
  ]
  rendererClasses [&NiagaraRibbonRendererProperties]
  renderers [{rendererClass=&NiagaraRibbonRendererProperties}]
 }
 Cake {
  emitterUpdateScript [
   &EmitterState(Life Cycle Mode=::NewEnumerator1, Loop Behavior=::NewEnumerator1, Loop Delay=0.03, Loop Duration=0.3, UseLoopDelay=true)
   &SpawnPerUnit(Module.Use Max Movement Threshold=true, Module.Use Movement Tolerance=false, Spawn Per Unit=75, Use Spawn Probability=false)
  ]
  particleSpawnScript [
   &FN_SetSpawnGroup(Number of Spawn Groups=2)
   &InitializeParticle(
    Lifetime Max=1
    Lifetime Min=0.5
    Lifetime Mode=::NewEnumerator1
    Mesh Scale Mode=::NewEnumerator3
    Mesh Uniform Scale=&Lerp_Float(
     A=0.01
     B=0.02
     Alpha=&FloatFromCurve(
      FloatCurve=DataInterface() {propertyValues {Curve {keys [
          {interpMode=RCIM_Cubic, tangentMode=RCTM_Break, time=0, value=0}
          {arriveTangent=2.5592, interpMode=RCIM_Cubic, tangentMode=RCTM_Break, time=1, value=1}
         ]}}}
      CurveIndex=&RandomRangeFloat(Minimum=0, Maximum=1)
     )
    )
    UsePositionOffset=false
    Write Material Random=true
   )
   &SocketLocation(SamplePosOffset=false, Socket Index=1, Socket Reader=$System.Socket Reader, Socket Sampling Mode=::NewEnumerator3, VelocitySpeedLimit=false)
   &BoxLocation(
    Box Origin:NiagaraPosition=$Output.SocketLocation.SampledPosition
    Box Size=<7 1 25>
    MaskSpawnGroup=false
    Module.Override Seed=false
    Rotate Box=false
    Surface Only=false
   )
   &AddVelocity(Velocity=&UniformRangedVector(Minimum=<0 -2 0>, Maximum=<0 5 0>))
   &AddVelocityInCone(
    Cone Angle=75
    Module.Use Velocity Falloff On Cone Axis=false
    Velocity Strength=&UniformRangedFloat(Minimum=1, Maximum=200)
   )
   &InheritVelocity(Module.Limit Speed=true, Velocity Scale=<1 1 0.5>) {module=InheritVelocity001}
   &AddVelocityFromPoint(
    Module.Use Falloff Distance=false
    Velocity Origin:Vector3f=&SimulationPosition()
    Velocity Strength=&UniformRangedFloat(Minimum=10, Maximum=100)
   )
   /SetVariables(Particles.MeshScaleRandom:NiagaraFloat=&UniformRangedFloat(Minimum=1, Maximum=2)) {isSetParameters=true, module=SetVariables_#2}
   &InitialMeshOrientation(Use Rotation Vector=false)
  ]
  particleUpdateScript [
   &ParticleState()
   &GravityForce()
   &SolveForcesAndVelocity(Clamp Velocity=false, Limit Acceleration=false)
   &ScaleMeshSize(Scale Factor=&VectorFromFloat(Value=&FloatFromCurve(
      FloatCurve=DataInterface() {propertyValues {Curve {keys [<at 0 1 RCIM_Cubic>,<at 0.2048 0.8914>,<at 1 0 RCIM_Cubic>]}}}
      Scale Curve=&Multiply_Float(A=$Particles.MeshScaleRandom, B=1.2)
     )))
   &ScaleColor(
    Scale Alpha=&FloatFromCurve(FloatCurve=DataInterface() {propertyValues {Curve {keys [
         {interpMode=RCIM_Cubic, tangentMode=RCTM_User, time=0, value=1}
         {arriveTangent=-0.3941, interpMode=RCIM_Cubic, leaveTangent=-0.3941, tangentMode=RCTM_User, time=1, value=0}
        ]}}})
    Scale RGB=&VectorFromFloat(Value=&FloatFromCurve(FloatCurve=DataInterface() {propertyValues {Curve {keys [
          {interpMode=RCIM_Cubic, tangentMode=RCTM_Break, time=0.6, value=1}
          {arriveTangent=-0.9546, interpMode=RCIM_Cubic, tangentMode=RCTM_Break, time=1, value=0}
         ]}}}))
    ScaleA=false
    ScaleRGB=true
   )
   &ScaleMeshSizeBySpeed(
    Max Scale Factor=<2 1.5 1>
    Speed Threshold:NiagaraFloat=&FloatFromCurve(FloatCurve=DataInterface() {propertyValues {Curve {keys [
         {arriveTangent=2294.4238, interpMode=RCIM_Cubic, leaveTangent=2294.4238, tangentMode=RCTM_User, time=0, value=150}
         {arriveTangent=78.1625, interpMode=RCIM_Cubic, leaveTangent=78.3263, tangentMode=RCTM_User, time=1, value=2000}
        ]}}})
   )
   &UpdateMeshOrientation(
    Rotation Rate=&RandomRangeFloat(Minimum=-2, Maximum=2)
    Rotation Vector=&RandomRangeVector(Minimum=<0 0 0>, Maximum=<1 1 1>)
   )
  ]
  rendererClasses [&NiagaraMeshRendererProperties]
  renderers [{rendererClass=&NiagaraMeshRendererProperties, rendererIndex=0}]
 }
 FlameTrail {
  emitterSpawnScript [&FN_RibbonIDOffsetUtility()]
  emitterUpdateScript [
   &EmitterState(Life Cycle Mode=::NewEnumerator1, Loop Behavior=::NewEnumerator1, Loop Duration=1, UseLoopDelay=false)
   &FN_SpawnAnimTrail(Spawn Interval=0)
  ]
  particleSpawnScript [
   &FN_InitializeAnimTrailBoneIndecies(Second Bone or Socket Index=0) {module=FN_InitializeAnimTrailBoneIndecies001}
   &FN_InitializeAnimTrail(AnimTrailWidth=3, Lifetime=0.3, Write Material Random=false)
  ]
  particleUpdateScript [
   &FN_UpdateAnimTrail(Locked Particle RibbonWidth=1)
   &DynamicMaterialParameters(
    Index 0 Param 3=&Multiply_Float(A=$System.Age, B=-0.7)
    Module.Index0Param0Write=true
    Module.Index0Param1Write=true
    Module.Index0Param2Write=true
    Module.Index0Param3Write=true
   )
   &ScaleRibbonWidth(Ribbon Width Scale=&FloatFromCurve(FloatCurve=DataInterface() {propertyValues {Curve {keys [{arriveTangent=28.9769, time=0, value=1}, {leaveTangent=22.6888, time=0.5, value=2}]}}}))
   &ScaleColor(
    Scale Alpha=&FloatFromCurve(FloatCurve=DataInterface() {propertyValues {Curve {keys [<at 0 1>,<at 0.2 0>]}}}, CurveIndex=$Particles.Age)
    Scale RGB=&Multiply_Vector(
     A=&MakeVectorFromLinearColorRGB(Color=&LerpLinearColors(
       StartColor=$User.RibbonColorLight
       EndColor=&LerpLinearColors(
        StartColor=$User.RibbonColorMiddle
        EndColor=$User.RibbonColorDark
        LerpFactor=&FloatFromCurve(FloatCurve=DataInterface() {propertyValues {Curve {keys [{interpMode=RCIM_Cubic, leaveTangent=0.2396, tangentMode=RCTM_Break, time=0.2, value=0}, <at 0.4 1>]}}})
       )
       LerpFactor=&FloatFromCurve(FloatCurve=DataInterface() {propertyValues {Curve {keys [{interpMode=RCIM_Cubic, leaveTangent=7.1493, tangentMode=RCTM_Break, time=0, value=0}, <at 0.2 1>]}}})
      ))
     B=&MakeVectorFromLinearColorRGB(Color=&ColorFromCurve(
       ColorCurve=DataInterface() {propertyValues {
         AlphaCurve {keys [<at 0 1>,<at 0.125 1>,<at 0.25 1>,<at 0.375 1>,<at 0.5 1>,<at 0.625 1>,<at 0.75 1>,<at 0.875 1>,<at 1 1>]}
         BlueCurve {keys [<at 0.1209 0.651>,<at 0.166 0.6172>,<at 0.208 0.5395>,<at 0.25 0.6172>,<at 0.291 0.6939>,<at 0.333 0.7379>,<at 0.375 0.7835>,<at 0.416 0.7991>,<at 0.458 0.8148>,<at 0.5 0.8308>,<at 0.541 0.8148>,<at 0.583 0.7991>,<at 0.625 0.7991>,<at 0.666 0.7529>,<at 0.708 0.7084>,<at 0.75 0.5711>,<at 0.791 0.4735>,<at 0.833 0.3231>,<at 0.909 0.651>]}
         GreenCurve {keys [<at 0.1209 0.4633>,<at 0.166 0.5333>,<at 0.208 0.5271>,<at 0.25 0.6376>,<at 0.291 0.7529>,<at 0.333 0.7835>,<at 0.375 0.8148>,<at 0.416 0.7991>,<at 0.458 0.8148>,<at 0.5 0.8469>,<at 0.541 0.7991>,<at 0.583 0.7991>,<at 0.625 0.8148>,<at 0.666 0.7991>,<at 0.708 0.7529>,<at 0.75 0.5711>,<at 0.791 0.5029>,<at 0.833 0.5029>,<at 0.909 0.4633>]}
         RedCurve {keys [<at 0.1209 0.234>,<at 0.166 0.7084>,<at 0.208 0.6867>,<at 0.25 0.8148>,<at 0.291 0.8879>,<at 0.333 0.8963>,<at 0.375 0.9131>,<at 0.416 0.8714>,<at 0.458 0.8714>,<at 0.5 0.9131>,<at 0.541 0.8632>,<at 0.583 0.8714>,<at 0.625 0.9131>,<at 0.666 0.8963>,<at 0.708 0.8714>,<at 0.75 0.7529>,<at 0.791 0.7529>,<at 0.833 0.8469>,<at 0.909 0.234>]}
        }}
       CurveIndex=Linked() {linkedVariable {name=NPC.FortniteNPC.TimeOfDay_0-1, type=NiagaraFloat}}
      ))
    )
    ScaleA=true
    ScaleRGB=true
   )
   /SetVariables(Particles.RibbonWidth_2:NiagaraFloat=&FloatFromCurve(
     FloatCurve=DataInterface() {propertyValues {Curve {keys [{arriveTangent=-10.0213, time=0.05, value=2}, <at 0.12 0>]}}}
     CurveIndex=$Particles.Age
    )) {isSetParameters=true}
  ]
  rendererClasses [&NiagaraRibbonRendererProperties]
  renderers [
   {rendererClass=&NiagaraRibbonRendererProperties}
   {rendererClass=&NiagaraRibbonRendererProperties, rendererIndex=1}
  ]
 }
 Sprinkles {
  emitterUpdateScript [
   &EmitterState(Life Cycle Mode=::NewEnumerator1, Loop Behavior=::NewEnumerator1, Loop Delay=0.03, Loop Duration=0.3, UseLoopDelay=true)
   &SpawnPerUnit(Module.Use Max Movement Threshold=true, Module.Use Movement Tolerance=false, Spawn Per Unit=50, Use Spawn Probability=false)
  ]
  particleSpawnScript [
   &FN_SetSpawnGroup(Number of Spawn Groups=2)
   &InitializeParticle(
    Color=&MergeHelper_ColorSwitch(
     0=$User.RibbonColorLight
     1=$User.RibbonColorMiddle
     2=$User.RibbonColorDark
     3=?
     Index=&RandomRangeInt(Minimum=0, Maximum=2)
     Num=::NewEnumerator1
    )
    Lifetime Max=1
    Lifetime Min=0.5
    Lifetime Mode=::NewEnumerator1
    Mesh Scale=&Multiply_VectorByFloat(Vector=<0.3 0.3 1>, Float=&RandomRangeFloat(Minimum=0.2, Maximum=0.3))
    Mesh Scale Mode=::NewEnumerator4
    UsePositionOffset=false
    Write Material Random=true
   )
   &SocketLocation(SamplePosOffset=false, Socket Index=1, Socket Reader=$System.Socket Reader, Socket Sampling Mode=::NewEnumerator3, VelocitySpeedLimit=false)
   &BoxLocation(
    Box Origin:NiagaraPosition=$Output.SocketLocation.SampledPosition
    Box Size=<7 1 25>
    MaskSpawnGroup=false
    Module.Override Seed=false
    Rotate Box=false
    Surface Only=false
   )
   &AddVelocity(Velocity=&UniformRangedVector(Minimum=<0 -2 0>, Maximum=<0 5 0>))
   &AddVelocityInCone(
    Cone Angle=75
    Module.Use Velocity Falloff On Cone Axis=false
    Velocity Strength=&UniformRangedFloat(Minimum=1, Maximum=200)
   )
   &InheritVelocity(Module.Limit Speed=true, Velocity Scale=<1 1 0.5>) {module=InheritVelocity001}
   &AddVelocityFromPoint(
    Module.Use Falloff Distance=false
    Velocity Origin:Vector3f=&SimulationPosition()
    Velocity Strength=&UniformRangedFloat(Minimum=10, Maximum=100)
   )
   /SetVariables(Particles.MeshScaleRandom:NiagaraFloat=&UniformRangedFloat(Minimum=1, Maximum=2)) {isSetParameters=true, module=SetVariables_#2}
   &InitialMeshOrientation(Use Rotation Vector=false)
  ]
  particleUpdateScript [
   &ParticleState()
   &GravityForce()
   &SolveForcesAndVelocity(Clamp Velocity=false, Limit Acceleration=false)
   &ScaleMeshSize(Scale Factor=&VectorFromFloat(Value=&FloatFromCurve(
      FloatCurve=DataInterface() {propertyValues {Curve {keys [<at 0 1 RCIM_Cubic>,<at 0.2048 0.8914>,<at 1 0 RCIM_Cubic>]}}}
      Scale Curve=&Multiply_Float(A=$Particles.MeshScaleRandom, B=1.2)
     )))
   &ScaleColor(
    Scale Alpha=&FloatFromCurve(FloatCurve=DataInterface() {propertyValues {Curve {keys [
         {interpMode=RCIM_Cubic, tangentMode=RCTM_User, time=0, value=1}
         {arriveTangent=-0.3941, interpMode=RCIM_Cubic, leaveTangent=-0.3941, tangentMode=RCTM_User, time=1, value=0}
        ]}}})
    Scale RGB=&VectorFromFloat(Value=&FloatFromCurve(FloatCurve=DataInterface() {propertyValues {Curve {keys [
          {interpMode=RCIM_Cubic, tangentMode=RCTM_Break, time=0.6, value=1}
          {arriveTangent=-0.9546, interpMode=RCIM_Cubic, tangentMode=RCTM_Break, time=1, value=0}
         ]}}}))
    ScaleA=false
    ScaleRGB=true
   )
   &ScaleMeshSizeBySpeed(
    Max Scale Factor=<2 1.5 1>
    Speed Threshold:NiagaraFloat=&FloatFromCurve(FloatCurve=DataInterface() {propertyValues {Curve {keys [
         {arriveTangent=2294.4238, interpMode=RCIM_Cubic, leaveTangent=2294.4238, tangentMode=RCTM_User, time=0, value=150}
         {arriveTangent=78.1625, interpMode=RCIM_Cubic, leaveTangent=78.3263, tangentMode=RCTM_User, time=1, value=2000}
        ]}}})
   )
   &UpdateMeshOrientation(
    Rotation Rate=&RandomRangeFloat(Minimum=-6, Maximum=6)
    Rotation Vector=&RandomRangeVector(Minimum=<0 0 0>, Maximum=<1 1 1>)
   )
  ]
  rendererClasses [&NiagaraMeshRendererProperties]
  renderers [{rendererClass=&NiagaraMeshRendererProperties, rendererIndex=0}]
 }
}
summary {
 systemName=NS_Pickaxe_Birthday_Trail
 emitters [
  {emitterName=Sprinkles, rendererClasses [&NiagaraMeshRendererProperties]}
  {emitterName=Cake, rendererClasses [&NiagaraMeshRendererProperties]}
  {emitterName=AnimTrail, rendererClasses [&NiagaraRibbonRendererProperties]}
  {emitterName=FlameTrail, rendererClasses [&NiagaraRibbonRendererProperties]}
 ]
 userVariables [
  User.MICake {
   type=MaterialInterface
   description=""
   defaultValue {
    struct=NiagaraExt_VariableValue_Object
    value {object=/BRCosmetics/Weapons/FORT_Melee/Pickaxe_FN_Birthday_Male/FX/MI_Pickaxe_FNBirthday_Cake7, objectClass=/Script/Engine.MaterialInterface}
   }
  }
  User.MIRibbon {
   type=MaterialInterface
   description=""
   defaultValue {
    struct=NiagaraExt_VariableValue_Object
    value {
     object=/BRCosmetics/Weapons/FORT_Melee/Pickaxe_FN_Birthday_Male/FX/MI_PickAxe_FNBirthday_Trail7
     objectClass=&MaterialInterface
    }
   }
  }
  User.RibbonColorDark {type=LinearColor, description="", defaultValue {struct=LinearColor, value=<rgba 0.0886 0 0.1651 1>}}
  User.RibbonColorLight {type=LinearColor, description="", defaultValue {struct=LinearColor, value=<rgba 1.9493 0.0919 2 1>}}
  User.RibbonColorMiddle {type=LinearColor, description="", defaultValue {struct=LinearColor, value=<rgba 0.3864 0.0081 0.2544 1>}}
  User.RibbonIDOffset {type=NiagaraInt32}
 ]
}
